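# Sound cue for Ego henchman voice lines played on full alert, per its name.
# A Random node picks among six equally weighted FullAlert wave variations.
# It uses the shared NPC alerted-voice attenuation, the non-critical enemy VO sound class and the henchmen combat chatter concurrency, with a slight volume boost.
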SoundCue /Clyde/Athena/Audio/VO/Henchmen/Ego/Cues/VO_Henchmen_Ego_S32_FullAlert_Cue
root=Random_0
nodes [
 {id=Random_0, children [WavePlayer_0,WavePlayer_1,WavePlayer_2,WavePlayer_3,WavePlayer_4,WavePlayer_5], properties {weights [1,1,1,1,1,1]}}
 {id=WavePlayer_0, properties {soundWaveAssetPtr=/Clyde/Athena/Audio/VO/Henchmen/Ego/Waves/VO_Henchmen_Ego_S32_FullAlert_010}}
 {id=WavePlayer_1, properties {soundWaveAssetPtr=/Clyde/Athena/Audio/VO/Henchmen/Ego/Waves/VO_Henchmen_Ego_S32_FullAlert_020}}
 {id=WavePlayer_2, properties {soundWaveAssetPtr=/Clyde/Athena/Audio/VO/Henchmen/Ego/Waves/VO_Henchmen_Ego_S32_FullAlert_030}}
 {id=WavePlayer_3, properties {soundWaveAssetPtr=/Clyde/Athena/Audio/VO/Henchmen/Ego/Waves/VO_Henchmen_Ego_S32_FullAlert_040}}
 {id=WavePlayer_4, properties {soundWaveAssetPtr=/Clyde/Athena/Audio/VO/Henchmen/Ego/Waves/VO_Henchmen_Ego_S32_FullAlert_050}}
 {id=WavePlayer_5, properties {soundWaveAssetPtr=/Clyde/Athena/Audio/VO/Henchmen/Ego/Waves/VO_Henchmen_Ego_S32_FullAlert_060}}
]
properties {
 attenuationSettings=/Game/Sounds/Fort_Human_VO_BR/NPCs/Shared/NPC_VO_Alerted_Att
 maxDistance=12500
 soundClassObject=/Game/Sounds/FortSoundClassesAndModes/Fort_VO_Enemy_NonCritical
 volumeMultiplier=1.1147
 concurrencySet [/Game/Sounds/Fort_Human_VO_BR/NPCs/Shared/NPC_Henchmen_Combat_Chatter_Conc]
}
